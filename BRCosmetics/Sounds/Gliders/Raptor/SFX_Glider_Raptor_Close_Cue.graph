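# Sound cue for the Raptor glider's close sound, a cosmetic glider effect.
# A single wave player plays the glider close wave directly, with no randomization or mixing nodes.
# It uses the Athena glider open/close attenuation and concurrency, the glider-land sound class, a cosmetic glider-close property sheet, and reduced volume.
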SoundCue /BRCosmetics/Sounds/Gliders/Raptor/SFX_Glider_Raptor_Close_Cue
root=WavePlayer_0
nodes [{id=WavePlayer_0, properties {soundWaveAssetPtr=&/SFX_Glider_Raptor_Close}}]
properties {
 attenuationSettings=/Game/Sounds/Attenuation/Fort_Player_Glider_OpenClose_Athena_Att
 audioPropertiesSheet=/BRCosmetics/Sounds/Settings/PropertySheets/Glider/APS_Cosmetic_Glider_Close
 maxDistance=10000
 soundClassObject=&Fort_Gliders_Land
 volumeMultiplier=0.5
 concurrencySet [&Athena_GliderOpenClose_Conc]
}
