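# Sound cue for the Raptor glider's thrust sound effect.
# A Random node picks between two thrust wave variations with equal weight.
# Playback runs at half volume, with shared glider thrust attenuation and concurrency, a cosmetic glider loop audio properties sheet, and the Fort_Gliders_Land sound class.
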SoundCue /BRCosmetics/Sounds/Gliders/Raptor/SFX_Glider_Raptor_Thrust_Cue
root=Random_1
nodes [
 {id=Random_1, children [WavePlayer_1,WavePlayer_6], properties {weights [1,1]}}
 {id=WavePlayer_1, properties {soundWaveAssetPtr=&/SFX_Glider_Raptor_Thrust_01}}
 {id=WavePlayer_6, properties {soundWaveAssetPtr=&/SFX_Glider_Raptor_Thrust_02}}
]
properties {
 attenuationSettings=&Fort_Player_Glider_Thrust_Athena_Att
 audioPropertiesSheet=&APS_Cosmetic_Glider_Loop
 maxDistance=12500
 soundClassObject=&Fort_Gliders_Land
 volumeMultiplier=0.5
 concurrencySet [&Athena_Glider_Thrust_Conc]
}
